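# Control Rig for the Thrive Spirit female pickaxe that procedurally animates the snake and tongue bones on the axe.
# The forward solve runs a sequence of Chain Harmonics units, each using only wave motion with its own tuning, pendulum and reach disabled.
# Midway, an Aim unit points bone dyn_snake_l_06 at the snake_aim_l control.
# FABRIK, item harmonics, math and SnakeHead_Ctrl transform nodes sit unconnected to execution.
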ControlRigBlueprint /BRCosmetics/Weapons/FORT_Melee/Pickaxe_Thrive_Spirit_Female/Meshes/Thrive_Spirit_Axe_CtrlRig
graphs [
 RigVMFunctionLibrary {links [], nodes []}
 RigVMModel {nodes [
   n1 = &RigUnit_BeginExecution()
   /Script/ControlRig.RigUnit_ChainHarmonics(
    ChainRoot=dyn_snake_r_01
    DrawWorldOffset=(Rotation=(X=0,Y=0,Z=0,W=1),Translation=(X=0,Y=0,Z=0),Scale3D=(X=1,Y=1,Z=1))
    Pendulum=(bEnabled=False,PendulumStiffness=2,PendulumGravity=(X=0,Y=0,Z=0),PendulumBlend=0.75,PendulumDrag=0.9,PendulumMinimum=0,PendulumMaximum=1,PendulumEase=Linear,UnwindAxis=(X=0,Y=1,Z=0),UnwindMinimum=0.2,UnwindMaximum=0.05)
    Reach=(bEnabled=False,ReachTarget=(X=0,Y=0,Z=0),ReachAxis=(X=1,Y=0,Z=0),ReachMinimum=0,ReachMaximum=0,ReachEase=SineEaseInOut)
    Speed=(X=0.3,Y=0.15,Z=0.1)
    Wave=(bEnabled=True,WaveFrequency=(X=10,Y=10,Z=10),WaveAmplitude=(X=0.25,Y=0.25,Z=2),WaveOffset=(X=1,Y=0,Z=2),WaveNoise=(X=8,Y=8,Z=15),WaveMinimum=0,WaveMaximum=0.75,WaveEase=SineEaseIn)
    WaveCurve="(EditorCurveData=(Keys=((InterpMode=RCIM_Linear,TangentMode=RCTM_Auto,TangentWeightMode=RCTWM_WeightedNone,Time=0,Value=0,ArriveTangent=0,ArriveTangentWeight=0,LeaveTangent=0,LeaveTangentWeight=0),(InterpMode=RCIM_Linear,TangentMode=RCTM_Auto,TangentWeight ...(truncated)"
    bDrawDebug=True
    ExecutePin=@n1.ExecutePin
   ) {is=struct RigVMUnitNode}
   > &RigUnit_ChainHarmonics(
    ChainRoot=dyn_snake_l_01
    DrawWorldOffset=(Rotation=(X=0,Y=0,Z=0,W=1),Translation=(X=0,Y=0,Z=0),Scale3D=(X=1,Y=1,Z=1))
    Pendulum=(bEnabled=False,PendulumStiffness=2,PendulumGravity=(X=0,Y=0,Z=0),PendulumBlend=0.75,PendulumDrag=0.98,PendulumMinimum=0,PendulumMaximum=0.1,PendulumEase=Linear,UnwindAxis=(X=0,Y=1,Z=0),UnwindMinimum=0.2,UnwindMaximum=0.05)
    Reach=(bEnabled=False,ReachTarget=(X=0,Y=0,Z=0),ReachAxis=(X=5,Y=0,Z=0),ReachMinimum=0,ReachMaximum=0,ReachEase=SineEaseInOut)
    Speed=(X=0.2,Y=0.2,Z=0.2)
    Wave=(bEnabled=True,WaveFrequency=(X=2.5,Y=6,Z=0.2),WaveAmplitude=(X=0.25,Y=0.5,Z=2),WaveOffset=(X=1,Y=0,Z=2),WaveNoise=(X=6,Y=6,Z=6),WaveMinimum=0,WaveMaximum=0.75,WaveEase=SineEaseIn)
    WaveCurve="(EditorCurveData=(Keys=((InterpMode=RCIM_Linear,TangentMode=RCTM_Auto,TangentWeightMode=RCTWM_WeightedNone,Time=-0.0198,Value=0.4929,ArriveTangent=0,ArriveTangentWeight=0,LeaveTangent=0,LeaveTangentWeight=0),(InterpMode=RCIM_Linear,TangentMode=RCTM_Auto,Ta ...(truncated)"
    bDrawDebug=True
   )
   > &RigUnit_ChainHarmonics(
    ChainRoot=dyn_snake_l2_01
    DrawWorldOffset=(Rotation=(X=0,Y=0,Z=0,W=1),Translation=(X=0,Y=0,Z=0),Scale3D=(X=1,Y=1,Z=1))
    Pendulum=(bEnabled=False,PendulumStiffness=2,PendulumGravity=(X=0,Y=0,Z=0),PendulumBlend=0.75,PendulumDrag=0.98,PendulumMinimum=0,PendulumMaximum=0.1,PendulumEase=Linear,UnwindAxis=(X=0,Y=1,Z=0),UnwindMinimum=0.2,UnwindMaximum=0.05)
    Reach=(bEnabled=False,ReachTarget=(X=0,Y=0,Z=0),ReachAxis=(X=5,Y=0,Z=0),ReachMinimum=0,ReachMaximum=0,ReachEase=SineEaseInOut)
    Speed=(X=0.1,Y=0.1,Z=0.1)
    Wave=(bEnabled=True,WaveFrequency=(X=2.5,Y=10,Z=0.2),WaveAmplitude=(X=0.15,Y=0.5,Z=2),WaveOffset=(X=1,Y=0,Z=2),WaveNoise=(X=5,Y=5,Z=5),WaveMinimum=0,WaveMaximum=0.75,WaveEase=SineEaseIn)
    WaveCurve="(EditorCurveData=(Keys=((InterpMode=RCIM_Linear,TangentMode=RCTM_Auto,TangentWeightMode=RCTWM_WeightedNone,Time=-0.0198,Value=0.4929,ArriveTangent=0,ArriveTangentWeight=0,LeaveTangent=0,LeaveTangentWeight=0),(InterpMode=RCIM_Linear,TangentMode=RCTM_Auto,Ta ...(truncated)"
    bDrawDebug=True
   )
   > /Script/ControlRig.RigUnit_AimItem(
    DebugSettings=(Scale=10,WorldOffset=(Rotation=(X=0,Y=0,Z=0,W=1),Translation=(X=0,Y=0,Z=0),Scale3D=(X=1,Y=1,Z=1)))
    Item=(Type=Bone,Name="dyn_snake_l_06")
    Primary=(Weight=1,Axis=(X=1,Y=0,Z=0),Target=(X=0,Y=0,Z=0),Kind=Location,Space=(Type=Bone))
    Secondary=(Weight=0,Axis=(X=0,Y=0,Z=1),Target=(X=1,Y=0,Z=0),Kind=Location,Space=(Type=Bone,Name="b_AR_Root"))
    Weight=1
    Primary.Target=&RigUnit_GetTransform(Item=(Type=Control,Name="snake_aim_l")).Transform.Translation
   ) {is=struct RigVMUnitNode}
   > &RigUnit_ChainHarmonics(
    ChainRoot=dyn_snake_top_01
    DrawWorldOffset=(Rotation=(X=0,Y=0,Z=0,W=1),Translation=(X=0,Y=0,Z=0),Scale3D=(X=1,Y=1,Z=1))
    Pendulum=(bEnabled=False,PendulumStiffness=2,PendulumGravity=(X=0,Y=0,Z=0),PendulumBlend=0.75,PendulumDrag=0.98,PendulumMinimum=0,PendulumMaximum=0.1,PendulumEase=Linear,UnwindAxis=(X=0,Y=1,Z=0),UnwindMinimum=0.2,UnwindMaximum=0.05)
    Reach=(bEnabled=False,ReachTarget=(X=0,Y=0,Z=0),ReachAxis=(X=1,Y=0,Z=0),ReachMinimum=0,ReachMaximum=0,ReachEase=SineEaseInOut)
    Speed=(X=0.15,Y=0.15,Z=0.1)
    Wave=(bEnabled=True,WaveFrequency=(X=2,Y=0.2,Z=0.2),WaveAmplitude=(X=1,Y=0,Z=1.5),WaveOffset=(X=1,Y=0,Z=2),WaveNoise=(X=8,Y=8,Z=10),WaveMinimum=0,WaveMaximum=0.75,WaveEase=SineEaseIn)
    WaveCurve="(EditorCurveData=(Keys=((InterpMode=RCIM_Linear,TangentMode=RCTM_Auto,TangentWeightMode=RCTWM_WeightedNone,Time=-0.0198,Value=0.4929,ArriveTangent=0,ArriveTangentWeight=0,LeaveTangent=0,LeaveTangentWeight=0),(InterpMode=RCIM_Linear,TangentMode=RCTM_Auto,Ta ...(truncated)"
    bDrawDebug=True
   )
   > &RigUnit_ChainHarmonics(
    ChainRoot=dyn_snake_l_tounge
    DrawWorldOffset=(Rotation=(X=0,Y=0,Z=0,W=1),Translation=(X=0,Y=0,Z=0),Scale3D=(X=1,Y=1,Z=1))
    Pendulum=(bEnabled=False,PendulumStiffness=2,PendulumGravity=(X=0,Y=0,Z=0),PendulumBlend=0.75,PendulumDrag=0.98,PendulumMinimum=0,PendulumMaximum=0.1,PendulumEase=Linear,UnwindAxis=(X=0,Y=1,Z=0),UnwindMinimum=0.2,UnwindMaximum=0.05)
    Reach=(bEnabled=False,ReachTarget=(X=0,Y=0,Z=0),ReachAxis=(X=5,Y=0,Z=0),ReachMinimum=0,ReachMaximum=0,ReachEase=SineEaseInOut)
    Speed=(X=0.5,Y=3,Z=5)
    Wave=(bEnabled=True,WaveFrequency=(X=0.1,Y=1,Z=3),WaveAmplitude=(X=1,Y=3,Z=5),WaveOffset=(X=1,Y=0,Z=2),WaveNoise=(X=0,Y=0.5,Z=0.2),WaveMinimum=0.2,WaveMaximum=0.75,WaveEase=SineEaseIn)
    WaveCurve="(EditorCurveData=(Keys=((InterpMode=RCIM_Linear,TangentMode=RCTM_Auto,TangentWeightMode=RCTWM_WeightedNone,Time=-0.0198,Value=0.49,ArriveTangent=0,ArriveTangentWeight=0,LeaveTangent=0,LeaveTangentWeight=0),(InterpMode=RCIM_Linear,TangentMode=RCTM_Auto,Tang ...(truncated)"
    bDrawDebug=True
   )
   > &RigUnit_ChainHarmonics(
    ChainRoot=dyn_snake_r_tounge
    DrawWorldOffset=(Rotation=(X=0,Y=0,Z=0,W=1),Translation=(X=0,Y=0,Z=0),Scale3D=(X=1,Y=1,Z=1))
    Pendulum=(bEnabled=False,PendulumStiffness=2,PendulumGravity=(X=0,Y=0,Z=0),PendulumBlend=0.75,PendulumDrag=0.98,PendulumMinimum=0,PendulumMaximum=0.1,PendulumEase=Linear,UnwindAxis=(X=0,Y=1,Z=0),UnwindMinimum=0.2,UnwindMaximum=0.05)
    Reach=(bEnabled=False,ReachTarget=(X=0,Y=0,Z=0),ReachAxis=(X=5,Y=0,Z=0),ReachMinimum=0,ReachMaximum=0,ReachEase=SineEaseInOut)
    Speed=(X=0.2,Y=2,Z=4)
    Wave=(bEnabled=True,WaveFrequency=(X=1,Y=2,Z=2),WaveAmplitude=(X=1,Y=2,Z=4),WaveOffset=(X=1,Y=0,Z=2),WaveNoise=(X=0,Y=0.2,Z=0.5),WaveMinimum=0.5,WaveMaximum=0.6,WaveEase=SineEaseIn)
    WaveCurve="(EditorCurveData=(Keys=((InterpMode=RCIM_Linear,TangentMode=RCTM_Auto,TangentWeightMode=RCTWM_WeightedNone,Time=-0.0198,Value=0.4502,ArriveTangent=0,ArriveTangentWeight=0,LeaveTangent=0,LeaveTangentWeight=0),(InterpMode=RCIM_Linear,TangentMode=RCTM_Auto,Ta ...(truncated)"
    bDrawDebug=True
   )
   > &RigUnit_ChainHarmonics(
    ChainRoot=dyn_snake_top_tounge
    DrawWorldOffset=(Rotation=(X=0,Y=0,Z=0,W=1),Translation=(X=0,Y=0,Z=0),Scale3D=(X=1,Y=1,Z=1))
    Pendulum=(bEnabled=False,PendulumStiffness=2,PendulumGravity=(X=0,Y=0,Z=0),PendulumBlend=0.75,PendulumDrag=0.98,PendulumMinimum=0,PendulumMaximum=0.1,PendulumEase=Linear,UnwindAxis=(X=0,Y=1,Z=0),UnwindMinimum=0.2,UnwindMaximum=0.05)
    Reach=(bEnabled=False,ReachTarget=(X=0,Y=0,Z=0),ReachAxis=(X=5,Y=0,Z=0),ReachMinimum=0,ReachMaximum=0,ReachEase=SineEaseInOut)
    Speed=(X=0.2,Y=1.5,Z=2)
    Wave=(bEnabled=True,WaveFrequency=(X=2.5,Y=0.5,Z=5),WaveAmplitude=(X=1,Y=2.5,Z=3.5),WaveOffset=(X=1,Y=0,Z=2),WaveNoise=(X=1,Y=2,Z=2),WaveMinimum=0.3,WaveMaximum=0.8,WaveEase=SineEaseIn)
    WaveCurve="(EditorCurveData=(Keys=((InterpMode=RCIM_Linear,TangentMode=RCTM_Auto,TangentWeightMode=RCTWM_WeightedNone,Time=-0.0198,Value=0.4502,ArriveTangent=0,ArriveTangentWeight=0,LeaveTangent=0,LeaveTangentWeight=0),(InterpMode=RCIM_Linear,TangentMode=RCTM_Auto,Ta ...(truncated)"
    bDrawDebug=True
   )
   > &RigUnit_ChainHarmonics(
    ChainRoot=dyn_snake_bottom_tounge_01
    DrawWorldOffset=(Rotation=(X=0,Y=0,Z=0,W=1),Translation=(X=0,Y=0,Z=0),Scale3D=(X=1,Y=1,Z=1))
    Pendulum=(bEnabled=false,PendulumStiffness=2,PendulumGravity=(X=0,Y=0,Z=0),PendulumBlend=1,PendulumDrag=0.98,PendulumMinimum=0,PendulumMaximum=1,PendulumEase=Linear,UnwindAxis=(X=0,Y=1,Z=0),UnwindMinimum=0.2,UnwindMaximum=0.05)
    Reach=(bEnabled=False,ReachTarget=(X=0,Y=0,Z=0),ReachAxis=(X=5,Y=0,Z=0),ReachMinimum=0,ReachMaximum=0,ReachEase=SineEaseInOut)
    Speed=(X=0.2,Y=1.5,Z=4)
    Wave=(bEnabled=True,WaveFrequency=(X=0,Y=0.5,Z=3),WaveAmplitude=(X=0,Y=1,Z=3),WaveOffset=(X=1,Y=0,Z=-2),WaveNoise=(X=0,Y=0.1,Z=0.4),WaveMinimum=0.2,WaveMaximum=0.6,WaveEase=SineEaseIn)
    WaveCurve="(EditorCurveData=(Keys=((InterpMode=RCIM_Linear,TangentMode=RCTM_Auto,TangentWeightMode=RCTWM_WeightedNone,Time=-0.0003,Value=1,ArriveTangent=0,ArriveTangentWeight=0,LeaveTangent=0,LeaveTangentWeight=0),(InterpMode=RCIM_Linear,TangentMode=RCTM_Auto,Tangent ...(truncated)"
    bDrawDebug=True
   )
   /Script/RigVM.RigVMFunction_AlphaInterpVector(
    ClampMax=1
    InRange=(Min=-10,Max=10)
    InterpSpeedDecreasing=10
    InterpSpeedIncreasing=10
    OutRange=(Min=-10,Max=10)
    Scale=1
    Value=(X=0,Y=0,Z=0)
   ) {
    is=struct RigVMUnitNode
    template="AlphaInterp::Execute(in Value,in Scale,in Bias,in bMapRange,in InRange,in OutRange,in bClampResult,in ClampMin,in ClampMax,in bInterpResult,in InterpSpeedIncreasing,in InterpSpeedDecreasing,out Result)"
   }
   /Script/ControlRig.RigUnit_ChainHarmonicsPerItem(
    ChainRoot=(Type=Control,Name="snake_aim_l")
    DrawWorldOffset=(Rotation=(X=0,Y=0,Z=0,W=1),Translation=(X=0,Y=0,Z=0),Scale3D=(X=1,Y=1,Z=1))
    ExecutePin=()
    Pendulum=(bEnabled=True,PendulumStiffness=2,PendulumGravity=(X=0,Y=0,Z=0),PendulumBlend=0.75,PendulumDrag=0.98,PendulumMaximum=0.1,UnwindAxis=(X=0,Y=1,Z=0),UnwindMinimum=0.2,UnwindMaximum=0.05)
    Reach=(bEnabled=True,ReachTarget=(X=1,Y=1,Z=1),ReachAxis=(X=1,Y=1,Z=1),ReachMinimum=1,ReachMaximum=1)
    Speed=(X=1,Y=1,Z=1)
    Wave=(bEnabled=True,WaveFrequency=(X=1,Y=0.6,Z=0.8),WaveAmplitude=(X=1,Y=1,Z=1),WaveOffset=(X=1,Y=1,Z=1),WaveNoise=(X=1,Y=1,Z=1),WaveMinimum=1,WaveMaximum=1)
    WaveCurve=(EditorCurveData=(Keys=((),(Time=1,Value=1)),DefaultValue=340282346638528859811704183484516925440,PreInfinityExtrap=RCCE_Constant,PostInfinityExtrap=RCCE_Constant))
    bDrawDebug=True
   ) {is=struct RigVMUnitNode}
   RigUnit_GetTransform = &RigUnit_GetTransform(Item=(Type=Control,Name="SnakeHead_Ctrl"), bInitial=false)
   /Script/ControlRig.RigUnit_FABRIKPerItem(
    ExecutePin=()
    Items=(Keys=((Type=Bone,Name="dyn_snake_r_01"),(Type=Bone,Name="dyn_snake_r_02"),(Type=Bone,Name="dyn_snake_r_03")))
    MaxIterations=10
    Precision=1
    Weight=1
    bPropagateToChildren=True
    bSetEffectorTransform=True
    EffectorTransform=@RigUnit_GetTransform.Transform
   ) {is=struct RigVMUnitNode}
   /Script/ControlRig.RigUnit_ItemHarmonics(
    ExecutePin=()
    RotationOrder=YZX
    Targets=()
    WaveAmplitude=(X=0,Y=70,Z=0)
    WaveFrequency=(X=1,Y=0.6,Z=0.8)
    WaveMaximum=1
    WaveMinimum=0.5
    WaveNoise=(X=0,Y=0,Z=0)
    WaveOffset=(X=0,Y=1,Z=2)
    WaveSpeed=(X=1,Y=1,Z=1)
   ) {is=struct RigVMUnitNode}
   &RigUnit_ItemHarmonics(
    ExecutePin=()
    RotationOrder=YZX
    Targets=((Item=(Type=Control,Name="snake_aim_l"),Ratio=1),(Item=(Type=Bone,Name="b_AR_Root"),Ratio=1))
    WaveAmplitude=(X=20,Y=70,Z=20)
    WaveFrequency=(X=200,Y=22,Z=200)
    WaveMaximum=1
    WaveMinimum=0.5
    WaveNoise=(X=20,Y=20,Z=20)
    WaveOffset=(X=20,Y=1,Z=200)
    WaveSpeed=(X=2,Y=200,Z=200)
   )
   &RigVMFunction_MathFloatAdd() {template="Add::Execute(in A,in B,out Result)"}
   /Script/RigVM.RigVMFunction_MathFloatLerp(B=1) {is=struct RigVMUnitNode, template="Interpolate::Execute(in A,in B,in T,out Result)"}
   n2 = &RigVMFunction_MathFloatMul(B=0.1, A=/Script/RigVM.RigVMFunction_GetDeltaTime() {is=struct RigVMUnitNode}.Result) {template="Multiply::Execute(in A,in B,out Result)"}
   /Script/RigVM.RigVMFunction_MathVectorFromFloat(Value=&RigVMFunction_MathFloatMul(
     B=3
     A=/Script/RigVM.RigVMFunction_MathFloatSin(Value=@n2.Result) {is=struct RigVMUnitNode, template="Sin::Execute(in Value,out Result)"}.Result
    ) {template="Multiply::Execute(in A,in B,out Result)"}.Result) {is=struct RigVMUnitNode, template="Cast::Execute(in Value,out Result)"}
   /Script/RigVM.RigVMFunction_RandomVector(Maximum=1, Seed=217) {is=struct RigVMUnitNode, template="Random::Execute(in Seed,in Minimum,in Maximum,in Duration,out Result)"}
   RigUnit_SetTransform = &RigUnit_SetTransform(
    ExecutePin=()
    Item=(Type=Control,Name="SnakeHead_Ctrl")
    Value=(Rotation=(X=0.0678,Y=0.1978,Z=-0.4263,W=0.8801),Translation=(X=10.6701,Y=-3.7365,Z=58.1373),Scale3D=(X=1,Y=1,Z=1))
    Weight=1
    bInitial=true
   )
  ]}
]
hierarchy [
 b_AR_Root {loc=<0 0 0>, rot=<0 0 0 1>}
 Melee_TwoHanded_Handle {parent=b_AR_Root, loc=<0 0 0>, rot=<0 0 -0.7071 -0.7071>}
 dyn_snake_r_01 {parent=Melee_TwoHanded_Handle, loc=<-3.7538 -4.0112 67.6194>, rot=<-0.311 -0.2691 0.5964 -0.6893>}
 dyn_snake_r_02 {parent=dyn_snake_r_01, loc=<3.3225 0.1319 -0.203>, rot=<-0.0883 -0.3713 -0.0824 -0.9206>}
 dyn_snake_r_03 {parent=dyn_snake_r_02, loc=<3.919 -0.137 0.689>, rot=<-0.0422 0.3341 0.1195 -0.934>}
 dyn_snake_r_04 {parent=dyn_snake_r_03, loc=<3.3484 -0.3041 0.0962>, rot=<-0.0049 0.3758 0.2056 -0.9036>}
 dyn_snake_r_05 {parent=dyn_snake_r_04, loc=<3.7646 -0.8852 -0.0593>, rot=<-0.0364 -0.1694 0.1723 -0.9697>}
 dyn_snake_r_tounge {parent=dyn_snake_r_05, loc=<5.4413 0 0>, rot=<-0.0677 0.034 -0.0152 -0.997>}
 dyn_snake_r_tounge_02 {parent=dyn_snake_r_tounge, loc=<3.0643 -0.1881 1.2763>, rot=<-0.0351 0.4535 0.0445 -0.8894>}
 dyn_snake_l_01 {parent=Melee_TwoHanded_Handle, loc=<2.2879 11.9272 46.1142>, rot=<0.3027 -0.6028 -0.3313 -0.6598>}
 dyn_snake_l_02 {parent=dyn_snake_l_01, loc=<4.3319 0 0>, rot=<0.2746 0.5837 -0.2688 -0.7153>}
 dyn_snake_l_03 {parent=dyn_snake_l_02, loc=<2.4916 0 0>, rot=<-0.0754 0.4191 -0.2174 -0.8783>}
 dyn_snake_l_04 {parent=dyn_snake_l_03, loc=<2.8859 0 0>, rot=<-0.2263 0.0391 -0.2111 -0.9501>}
 dyn_snake_l_05 {parent=dyn_snake_l_04, loc=<3.5787 0 0>, rot=<-0.0321 -0.2422 -0.0453 -0.9686>}
 dyn_snake_l_06 {parent=dyn_snake_l_05, loc=<4.7537 0 0>, rot=<0.0158 -0.269 0.1662 -0.9486>}
 dyn_snake_l_tounge {parent=dyn_snake_l_06, loc=<6.3162 -0.1138 -0.4866>, rot=<-0.0089 0.0284 0.0848 -0.996>}
 dyn_snake_l_tounge_02 {parent=dyn_snake_l_tounge, loc=<3.2916 0 0>, rot=<-0.0677 0.0563 -0.0101 -0.9961>}
 dyn_snake_l2_01 {parent=Melee_TwoHanded_Handle, loc=<-3.1498 6.173 64.8851>, rot=<0.5654 -0.4726 -0.4194 -0.5301>}
 dyn_snake_l2_02 {parent=dyn_snake_l2_01, loc=<5.4201 0.0039 0.4385>, rot=<-0.0189 0.2961 -0.1004 -0.9497>}
 dyn_snake_l2_03 {parent=dyn_snake_l2_02, loc=<2.9175 0 0>, rot=<0.048 0.1971 -0.0417 -0.9783>}
 dyn_snake_l2_04 {parent=dyn_snake_l2_03, loc=<4.2518 0 0>, rot=<-0.0242 -0.0302 0.0837 -0.9957>}
 dyn_snake_top_01 {parent=Melee_TwoHanded_Handle, loc=<3.2666 3.1912 61.7041>, rot=<0.2373 0.6078 0.7577 0.0083>}
 dyn_snake_top_02 {parent=dyn_snake_top_01, loc=<2.9829 0 0>, rot=<-0.0714 -0.2373 0.0765 -0.9658>}
 dyn_snake_top_tounge {parent=dyn_snake_top_02, loc=<5.5653 0 0>, rot=<-0.0471 -0.1344 -0.155 -0.9776>}
 dyn_snake_top_tounge_02 {parent=dyn_snake_top_tounge, loc=<1.3375 0 0>, rot=<-0.0536 0.0348 0.0283 -0.9976>}
 dyn_snake_bottom_tounge_01 {parent=Melee_TwoHanded_Handle, loc=<-0.8261 -0.8679 14.1245>, rot=<-0.2798 0.1082 0.8898 0.344>}
 dyn_snake_bottom_tounge_02 {parent=dyn_snake_bottom_tounge_01, loc=<0.9013 -0.0451 -0.5356>, rot=<0.0313 0.0342 -0.0483 -0.9978>}
 dyn_snake_bottom_tounge_03 {parent=dyn_snake_bottom_tounge_02, loc=<1.4311 0 0>, rot=<-0.0411 -0.09 0.0563 -0.9935>}
 dyn_snake_bottom_tounge_04 {parent=dyn_snake_bottom_tounge_03, loc=<1.6774 0 0>, rot=<0.1363 0.1284 -0.2042 -0.9609>}
 dyn_snake_bottom_tounge_05 {parent=dyn_snake_bottom_tounge_04, loc=<1.2021 0 0>, rot=<-0.0138 -0.1232 0.0211 -0.9921>}
 snake_aim_l {controlType=EULER_TRANSFORM, parent=b_AR_Root, type=Control, loc=<0 0 0>, rot=<0 0 0 1>}
]
