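# Animation Blueprint for the Thrive Spirit pickaxe skeleton that procedurally animates its snakes.
# Each update sets a Head R vector from a sine of game time, which drives the Control Rig's snake_aim_l transform as a location.
# The rig output passes through two partially weighted Trail controllers that add lag to the right and left snake tongue chains.
# A third Trail controller for dyn_snake_r_05 is unconnected.
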AnimBlueprint /BRCosmetics/Weapons/FORT_Melee/Pickaxe_Thrive_Spirit_Female/Meshes/Thrive_Spirit_Axe_Skeleton_AnimBP
target_skeleton=&/Thrive_Spirit_Axe_Skeleton
graphs [
 AnimGraph {nodes [
   n1 = Get Head R() -> (Head R:Vector)
   n2 = Control Rig!(Source=Input Pose(), Alpha, snake_aim_l:Transform="To Transform (Vector)"!(InLocation:Vector=@n1) -> (ReturnValue:Transform)) -> (Pose) {
    alphaBoolBlend {blendOption=Linear}
    controlRigAssetReference {blueprintRigClass=&/Thrive_Spirit_Axe_CtrlRig.Thrive_Spirit_Axe_CtrlRig_C}
   }
   n3 = Trail controller - Bone: dyn_snake_r_tounge_02(ComponentPose=Local To Component(LocalPose=@n2), Alpha=0.6) {
    bShowBaseMotion=false
    bShowLimit=false
    chainLength=4
    relaxationSpeedScale=1.5
    stretchLimit=0.001
    relaxationSpeedScaleInputProcessor {bias=1, interpSpeedIncreasing=30}
    rotationLimits [
     {limitMax=<30 30 30>, limitMin=<-30 -30 -30>}
     {limitMax=<30 30 30>, limitMin=<-30 -30 -30>}
     {limitMax=<30 30 30>, limitMin=<-30 -30 -30>}
     {limitMax=<30 30 30>, limitMin=<-30 -30 -30>}
    ]
    rotationOffsets [<0 0 0>, <0 0 0>, <0 0 0>, <0 0 0>]
    trailBone {boneName=dyn_snake_r_tounge_02}
    trailRelaxationSpeed {editorCurveData {keys [<at 0 45 RCIM_Linear>,<at 1 3 RCIM_Linear>]}}
   }
   n4 = Trail controller - Bone: dyn_snake_l_tounge_02(ComponentPose=@n3, Alpha=0.5) {
    bLimitRotation=true
    bShowBaseMotion=false
    bShowLimit=false
    chainLength=4
    relaxationSpeedScale=1.5
    stretchLimit=1.5
    alphaScaleBiasClamp {interpSpeedDecreasing=1, interpSpeedIncreasing=1}
    baseJoint {boneName=b_AR_Root}
    planarLimits [{drivingBone {boneName=None}, planeTransform {location=<-15 -5 35>, rotation=<rot 0 0 0>, scale=<0.2 0.2 0.2>}}]
    relaxationSpeedScaleInputProcessor {bias=1, interpSpeedIncreasing=30}
    rotationLimits [
     {limitMax=<20 20 20>, limitMin=<-20 -20 -20>}
     {limitMax=<20 20 20>, limitMin=<-20 -20 -20>}
     {limitMax=<50 50 50>, limitMin=<-50 -50 -50>}
     {limitMax=<50 50 50>, limitMin=<-50 -50 -50>}
    ]
    rotationOffsets [<0 0 0>, <0 0 0>, <0 0 0>, <0 0 0>]
    trailBone {boneName=dyn_snake_l_tounge_02}
    trailRelaxationSpeed {editorCurveData {keys [<at 0 30 RCIM_Linear>,<at 1 3 RCIM_Linear>]}}
   }
   Output Pose(Result=Component To Local(ComponentPose=@n4)) {layerGroup=None}
   Trail controller - Bone: dyn_snake_r_05(Alpha=1) {
    chainLength=5
    rotationLimits [{limitMax=<180 180 180>}, {limitMax=<180 180 180>}, {limitMax=<180 180 180>}, {limitMax=<180 180 180>}, {limitMax=<180 180 180>}]
    rotationOffsets [<0 0 0>, <0 0 0>, <0 0 0>, <0 0 0>, <0 0 0>]
    trailBone {boneName=dyn_snake_r_05}
    bLimitStretch=~
   }
   vector * vector(A=@n1, B:Vector="1,1,1") -> (ReturnValue=~)
  ]}
 EventGraph {nodes [
   Event BlueprintUpdateAnimation()
   n1 = GetGameTimeInSeconds()
   n2 = float * float(A="Sin (Radians)"(A=float * float(A=@n1, B=1)), B=5) {is=title K2Node_PromotableOperator}
   > Set Head R(Head R:Vector=MakeVector(X=@n2, Y=float * float(A=@n2, B=0.3), Z=float * float(A=@n2, B=0.7))) -> (Output_Get:Vector)
   TryGetPawnOwner()
  ]}
]
variables [Head R {type=Vector}]
